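; AGV_PistolCanary is an audio gameplay volume that changes the sound mix for the PistolCanary event while it is toggled on.
; When cosmetic events can execute, toggling on pushes the Event Sound Mix Modifier and tells the PistolCanary event audio dispatcher that the listener entered.
; Toggling off pops the modifier and reports the listener exit.
; Both transitions print on-screen and log debug messages naming PistolCanary_POI_MixMod.
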(event EventBeginPlay)

(event Collision|EventActorBeginOverlap
  (params (OtherActor :type Actor%20Object%20Reference)))

(event EventTick
  (params (DeltaSeconds :type Float%20%28single-precision%29)))

(event OnToggledOn(AudioGameplayVolume)
  (Networking|CanExecuteCosmeticEvents
    (:True
      (Audio|PushSoundMixModifier (Variables|Default|GetEvent%20Sound%20Mix%20Modifier))
      (Development|PrintString "Activated PistolCanary_POI_MixMod" true true "(R=0.000000,G=0.660000,B=1.000000,A=1.000000)" 1e+01)
      (Class|PistolCanaryEventAudioDispatcher|ListenerEnterVolume (Variables|Default|GetPistolCanary_EventAudioDispatcher)))))

(event OnToggledOff(AudioGameplayVolume)
  (Networking|CanExecuteCosmeticEvents
    (:True
      (Audio|PopSoundMixModifier (Variables|Default|GetEvent%20Sound%20Mix%20Modifier))
      (Development|PrintString "Deactivated PistolCanary_POI_MixMod" true true "(R=0.000000,G=0.660000,B=1.000000,A=1.000000)" 1e+01)
      (Class|PistolCanaryEventAudioDispatcher|ListenerExitVolume (Variables|Default|GetPistolCanary_EventAudioDispatcher)))))

(fn ConstructionScript)
