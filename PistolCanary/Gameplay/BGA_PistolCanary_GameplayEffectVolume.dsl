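; Building game actor (BGA) volume that gives player pawns a configurable Gameplay Effect while they are inside it.
; On begin overlap of Volume, it casts to FortPlayerPawn and applies the effect to that pawn's ability system component.
; On end overlap, it removes all stacks of that effect whose instigator is the pawn's own component.
; Both events write debug messages to the log only.
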(event OnComponentBeginOverlap(Volume)
  (params (OverlappedComponent :type Primitive%20Component%20Object%20Reference) (OtherActor :type Actor%20Object%20Reference) (OtherComp :type Primitive%20Component%20Object%20Reference) (OtherBodyIndex :type Integer) (bFromSweep :type Boolean) (SweepResult :type Hit%20Result%20Structure%20%28by%20ref%29))
  (bind _asfort_player_pawn (Utilities|Casting|CastToFortPlayerPawn OtherActor))
  (Utilities|IsValid (Ability|GetFortAbilitySystemComponent _asfort_player_pawn)
    (:Is%20Valid
      (GameplayEffects|ApplyGameplayEffectToSelf :self (Ability|GetFortAbilitySystemComponent _asfort_player_pawn) :GameplayEffectClass (Variables|Default|GetGameplay%20Effect) :Level 0.0)
      (Development|PrintString "BGA: EFFECT STARTED" false true "(R=1.000000,G=0.286404,B=0.000000,A=1.000000)"))))

(event OnComponentEndOverlap(Volume)
  (params (OverlappedComponent :type Primitive%20Component%20Object%20Reference) (OtherActor :type Actor%20Object%20Reference) (OtherComp :type Primitive%20Component%20Object%20Reference) (OtherBodyIndex :type Integer))
  (bind _asfort_player_pawn (Utilities|Casting|CastToFortPlayerPawn OtherActor))
  (Utilities|IsValid (Ability|GetFortAbilitySystemComponent _asfort_player_pawn)
    (:Is%20Valid
      (GameplayEffects|RemoveActiveGameplayEffectbySourceEffect (Ability|GetFortAbilitySystemComponent _asfort_player_pawn) (Variables|Default|GetGameplay%20Effect) (Ability|GetFortAbilitySystemComponent _asfort_player_pawn) -1)
      (Development|PrintString "BGA: EFFECT ENDED" false true "(R=1.000000,G=0.286404,B=0.000000,A=1.000000)"))))

(fn ConstructionScript)
