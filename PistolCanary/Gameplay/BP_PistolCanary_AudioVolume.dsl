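; BP_PistolCanary_AudioVolume is an actor Blueprint in the PistolCanary gameplay folder, named as an audio volume.
; It contains only empty stubs for actor begin overlap, actor end overlap and the construction script.
; None of these has logic attached, so the graphs define no audio behavior of their own.
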(event Collision|EventActorBeginOverlap
  (params (OtherActor :type Actor%20Object%20Reference)))

(event Collision|EventActorEndOverlap
  (params (OtherActor :type Actor%20Object%20Reference)))

(fn ConstructionScript)
