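; BP_PistolCanary_BuildingDestroyer is a gameplay actor that clears Fortnite building pieces inside its box volume.
; A DestroyBuildings event stores a maximum lifespan, then, on the authority only, box-overlaps world static and dynamic objects for BuildingSMActor instances.
; Each found piece gets a random lifespan between a small minimum and that maximum, so the structures disappear at staggered times rather than at once.
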(event Custom|DestroyBuildings
  (params (LifespanMax :type Float%20%28double-precision%29))
  (Variables|Default|SetLifespanMax LifespanMax)
  (Networking|SwitchHasAuthority
    (:Authority
      (bind _outactors (Collision|BoxOverlapActors (Transformation|GetWorldLocation (Variables|Default|GetBox)) (Components|Box|GetScaledBoxExtent (Variables|Default|GetBox)) (Utilities|Array|MakeArray "ObjectTypeQuery1" "ObjectTypeQuery2") "/Script/FortniteGame.BuildingSMActor" :!class /Script/Engine.KismetSystemLibrary)) ; find structures
      (for _array_element _outactors
        (Actor|SetLifeSpan _array_element (Math|Random|RandomFloatinRange 0.05 (Variables|Default|GetLifespanMax)))))))

(fn ConstructionScript)
