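; Server-authoritative Pistol Canary event zone that alters players, vehicles and pickups inside it during a calendar-driven event.
; Calendar watcher or GM cheat events call ZoneEvent, toggling collision, broadcasting zone state and, outside post-event mode, toggling the respawn mutator.
; Overlap applies gravity, aura and damage-immunity effects, disables vehicles via loose tags, and destroys pickups outside post-event mode.
; Exit reverses player and vehicle effects, granting fall-damage immunity to falling players.
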(event EventBeginPlay
  ;[0 0.15 0.1039 0.5] Activate Zone via Calendar Event
  (Networking|SwitchHasAuthority
    (:Authority
      (Class|FortAsyncActionAuthorityMatchReady|MatchReadyAsync
        (:OnReady
          (MCPContext|InitialCalendarSyncCompleted
            (:OnSyncComplete
              (MCPContext|WatchCalendarEvent (Variables|Default|GetEventZoneCalendarEventName)
                (:OnEventActive
                  (CallFunction|ZoneEvent true))
                (:OnEventEnded
                  (CallFunction|ZoneEvent false))))))))))

(event Custom|GM_EventZoneOn
  ;[0 0.15 0.05764 0.5] Activate/Deactivate Zone via cheats
  (Networking|SwitchHasAuthority
    (:Authority
      (CallFunction|ZoneEvent true))))

(event Custom|GM_EventZoneOff
  ;[0 0.15 0.05764 0.5] Activate/Deactivate Zone via cheats
  (Networking|SwitchHasAuthority
    (:Authority
      (CallFunction|ZoneEvent false))))

(event OnComponentBeginOverlap(EventZone)
  (params (OverlappedComponent :type Primitive%20Component%20Object%20Reference) (OtherActor :type Actor%20Object%20Reference) (OtherComp :type Primitive%20Component%20Object%20Reference) (OtherBodyIndex :type Integer) (bFromSweep :type Boolean) (SweepResult :type Hit%20Result%20Structure%20%28by%20ref%29))
  ;[0 0.15 0.09335 0.5] Apply Zone effects when players begin overlap
  (Networking|SwitchHasAuthority
    (:Authority
      (if (and (Utilities|ClassIsChildOf (Utilities|GetClass OtherActor) "/Script/FortniteGame.FortPickupAthena") (not (Variables|Default|GetPostEventMode)))
        (Actor|DestroyActor OtherActor)
        (else
          (CallFunction|ApplyZoneEffectstoPlayer OtherActor)
          (Utilities|IsValid OtherActor
            (:Is%20Valid
              (CallFunction|DisableVehicle true OtherActor))))))))

(event OnComponentEndOverlap(EventZone)
  (params (OverlappedComponent :type Primitive%20Component%20Object%20Reference) (OtherActor :type Actor%20Object%20Reference) (OtherComp :type Primitive%20Component%20Object%20Reference) (OtherBodyIndex :type Integer))
  ;[0 0.15 0.09335 0.5] Apply Zone effects when players begin overlap
  (Networking|SwitchHasAuthority
    (:Authority
      (CallFunction|RemoveZoneEffectsfromPlayer OtherActor)
      (CallFunction|DisableVehicle false OtherActor))))

(event Custom|ZoneEvent
  (params (Activate :type Boolean))
  (bind _respawnmutator (Variables|Default|GetRespawnMutator))
  ;[0 0.15 0.09362 0.5] Activate Collision for Event Zone and set Zone Enabled to True.
  (Collision|SetCollisionEnabled (Variables|Default|GetEventZone) (select Activate "QueryOnly" "NoCollision"))
  (GameplayEventRouter|BroadcastStatefulEvent (GameplayEventRouter|GetGlobalEventRouter) (Utilities|Struct|MakePistolCanaryEventZoneActive Activate) 0)
  (if (Variables|Default|GetPostEventMode)
    _
    (elif Activate
      (Utilities|IsValid _respawnmutator
        (:Is%20Valid
          _)
        (:Is%20Not%20Valid
          (bind _returnvalue (Class|SpecialEventGameplayLibrary|SpawnMutator "/PistolCanary/Gameplay/M_SE_PC_Respawn.M_SE_PC_Respawn_C"))
          (Variables|Default|SetRespawnMutator _returnvalue)))
      (Mutator|SetMutatorActive _respawnmutator true)
      (bind _asm_se_pc_respawn (Utilities|Casting|CastToM_SE_PC_Respawn _respawnmutator))
      (Class|FortAthenaMutatorSERespawn|OverrideRespawnArea _asm_se_pc_respawn (Variables|Default|GetRespawn%20Area))
      (else
        (bind _returnvalue_2 (Mutator|GetGameplayMutatorbyClass "/PistolCanary/Gameplay/M_SE_PC_Respawn.M_SE_PC_Respawn_C"))
        (Mutator|SetMutatorActive _returnvalue_2 false)))))

(event Custom|DisableVehicle
  (params (Disable :type Boolean) (Actor :type Actor%20Object%20Reference))
  (bind _gameplay_tags (Variables|Default|GetGameplay%20Tags))
  (bind _asfort_athena_vehicle (Utilities|Casting|CastToFortAthenaVehicle Actor))
  (if Disable
    (Vehicle|WakeUp _asfort_athena_vehicle)
    (Components|Movement|StopMovementImmediately (Pawn|GetMovementComponent _asfort_athena_vehicle))
    (Ability|GameplayEffect|AddLooseGameplayTags Actor _gameplay_tags true)
    (else
      (Ability|GameplayEffect|RemoveLooseGameplayTags Actor _gameplay_tags true))))

(fn ApplyZoneEffectsToPlayer
  (params (Actor :type Actor%20Object%20Reference))
  (bind _returnvalue (Ability|GetFortAbilitySystemComponent Actor))
  ;[0 0.15 0.1302 0.5] Modify Gravity scale, and apply Zone GEs
  (bind _asfort_player_pawn_athena (Utilities|Casting|CastToFortPlayerPawnAthena Actor))
  (FortPawn|SetGravityMultiplier _asfort_player_pawn_athena (ScalableFloat|GetValueatLevelwithContext (Variables|Default|GetGravityScaleMultiplier) 1.0))
  (Utilities|IsValid _returnvalue
    (:Is%20Valid
      (GameplayEffects|ApplyGameplayEffectToTarget :self _returnvalue :GameplayEffectClass (select (Variables|Default|GetPostEventMode) (Variables|Default|GetPostEventAura) (Variables|Default|GetPreEventAura)) :Target _returnvalue :Level 0.0)
      (GameplayEffects|ApplyGameplayEffectToTarget :self _returnvalue :GameplayEffectClass "/PistolCanary/Gameplay/Effects/GE_PistolCanary_EventAuraFX.GE_PistolCanary_EventAuraFX_C" :Target _returnvalue :Level 0.0)
      (GameplayEffects|ApplyGameplayEffectToTarget :self _returnvalue :GameplayEffectClass "/PistolCanary/Gameplay/Effects/GE_PistolCanary_EventDamageImmune.GE_PistolCanary_EventDamageImmune_C" :Target _returnvalue :Level 0.0))
    (:Is%20Not%20Valid
      (Fortnite|FortShippingLog self "PistolCanaryEventZone: Overlapping Actor does not have a Fort Ability System Component."))))

(fn RemoveZoneEffectsFromPlayer
  (params (ActorEndingOverlap :type Actor%20Object%20Reference))
  (bind _returnvalue (Ability|GetFortAbilitySystemComponent ActorEndingOverlap))
  ;[0 0.15 0.1277 0.5] Remove Zone related Gameplay effects, and set Gravity scale to default.
  (bind _asfort_player_pawn_athena (Utilities|Casting|CastToFortPlayerPawnAthena ActorEndingOverlap))
  (FortPawn|SetGravityMultiplier _asfort_player_pawn_athena (Variables|Default|GetDefaultGravityScale))
  (Utilities|IsValid _returnvalue
    (:Is%20Valid
      (GameplayEffects|RemoveActiveGameplayEffectbySourceEffect _returnvalue "/PistolCanary/Gameplay/Effects/GE_PistolCanary_EventAuraFX.GE_PistolCanary_EventAuraFX_C" 0 -1)
      (GameplayEffects|RemoveActiveGameplayEffectbySourceEffect _returnvalue (Variables|Default|GetPostEventAura) 0 -1)
      (GameplayEffects|RemoveActiveGameplayEffectbySourceEffect _returnvalue (Variables|Default|GetPreEventAura) 0 -1)
      (GameplayEffects|RemoveActiveGameplayEffectbySourceEffect _returnvalue "/PistolCanary/Gameplay/Effects/GE_PistolCanary_EventDamageImmune.GE_PistolCanary_EventDamageImmune_C" 0 -1)
      (GameplayEffects|RemoveActiveGameplayEffectbySourceEffect _returnvalue "/PistolCanary/Gameplay/Effects/GE_PistolCanary_CombatAura.GE_PistolCanary_CombatAura_C" 0 -1)
      (GameplayEffects|RemoveActiveGameplayEffectbySourceEffect _returnvalue "/PistolCanary/Gameplay/Effects/GE_PistolCanary_EventAuraBots.GE_PistolCanary_EventAuraBots_C" 0 -1)
      (if (Utilities|Enum|Equal(Enum) (Variables|/Script/Engine.CharacterMovementComponent|GetMovementMode (Variables|/Script/Engine.Character|GetCharacterMovement _asfort_player_pawn_athena)) "MOVE_Falling")
        (GameplayEffects|ApplyGameplayEffectToTarget :self _returnvalue :GameplayEffectClass "/Game/Athena/Items/Parents/FallImmunity/GE_Athena_TillLandFallDamageImmunity_Grant_ApplyInAirOnly.GE_Athena_TillLandFallDamageImmunity_Grant_ApplyInAirOnly_C" :Target _returnvalue :Level 0.0)))))

(fn ConstructionScript)
