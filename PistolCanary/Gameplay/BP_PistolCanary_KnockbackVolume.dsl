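; Knockback volume for the PistolCanary event that launches players who hit or overlap its sphere and plays an impact sound.
; The sphere starts disabled; enabling it finds the player teleporter, then after a short delay teleports the local player if overlapping.
; Launch direction comes from the volume's own normalized world location, scaled and Z-clamped.
; Each knockback starts a failsafe timer that teleports a player still inside; end overlap clears it.
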(event EventBeginPlay
  (CallFunction|DisableKnockbackCollision))

(event Custom|DisableKnockbackCollision
  (Collision|SetCollisionEnabled (Variables|Default|GetSphere) "NoCollision"))

(event Custom|EnableKnockbackCollision
  (bind _returnvalue (Game|GetPlayerPawn 0))
  (Collision|SetCollisionEnabled (Variables|Default|GetSphere) "QueryOnly")
  (bind _returnvalue_2 (Actor|GetActorOfClass "/PistolCanary/Gameplay/BP_PistolCanary_PlayerTeleporter.BP_PistolCanary_PlayerTeleporter_C"))
  (Variables|Default|SetTeleporter _returnvalue_2)
  (Utilities|FlowControl|Delay 0.5)
  (if (Collision|IsOverlappingActor (Variables|Default|GetSphere) _returnvalue :!class /Script/Engine.PrimitiveComponent)
    (Utilities|Casting|CastToFortPlayerPawn _returnvalue)
    (CallFunction|TeleportPlayerifStillinArea)))

(event OnComponentHit(Sphere)
  (params (HitComponent :type Primitive%20Component%20Object%20Reference) (OtherActor :type Actor%20Object%20Reference) (OtherComp :type Primitive%20Component%20Object%20Reference) (NormalImpulse :type Vector :default "0, 0, 0") (Hit :type Hit%20Result%20Structure%20%28by%20ref%29))
  (bind _asfort_player_pawn (Utilities|Casting|CastToFortPlayerPawn OtherActor))
  (Fortnite|FortShippingLog self "Player hitting knockback volume")
  (CallFunction|KnockbackPlayer _asfort_player_pawn)
  (CallFunction|StartTeleportCheck))

(event Custom|StartTeleportCheck
  (bind _returnvalue (Utilities|Time|SetTimerbyEvent (delegate Custom|TeleportPlayerIfStillInArea) (Variables|Default|GetFailsafeTeleportCheckTime)))
  (Variables|Default|SetTimer_TeleportCheck _returnvalue))

(event OnComponentBeginOverlap(Sphere)
  (params (OverlappedComponent :type Primitive%20Component%20Object%20Reference) (OtherActor :type Actor%20Object%20Reference) (OtherComp :type Primitive%20Component%20Object%20Reference) (OtherBodyIndex :type Integer) (bFromSweep :type Boolean) (SweepResult :type Hit%20Result%20Structure%20%28by%20ref%29))
  (bind _asfort_player_pawn (Utilities|Casting|CastToFortPlayerPawn OtherActor))
  (Variables|Default|SetPlayer _asfort_player_pawn)
  (Fortnite|FortShippingLog self "Player overlapping knockback volume")
  (CallFunction|KnockbackPlayer (Variables|Default|GetPlayer))
  (CallFunction|StartTeleportCheck))

(event Custom|TeleportPlayerIfStillInArea
  (bind _player (Variables|Default|GetPlayer))
  (Fortnite|FortShippingLog 0 "PistolCanary attempting player teleport")
  (Utilities|IsValid _player
    (:Is%20Valid
      (if (Collision|IsOverlappingActor (Variables|Default|GetSphere) _player :!class /Script/Engine.PrimitiveComponent)
        (Class|BPPistolCanaryPlayerTeleporter|TeleportNextPawn (Variables|Default|GetTeleporter) _player)))))

(event Custom|KnockbackPlayer
  (params (Player :type Fort%20Pawn%20Object%20Reference))
  (bind (_returnvalue_x _returnvalue_y _returnvalue_z) (Math|Vector|BreakVector (Math|Vector|MakeVector (Variables|Default|GetDistanceMultiplier) (Variables|Default|GetDistanceMultiplier) (Variables|Default|GetDistanceMultiplier))))
  (bind (_returnvalue_x_2 _returnvalue_y_2 _returnvalue_z_2) (Math|Vector|BreakVector (* (Math|Vector|Normalize (Transformation|GetActorLocation)) (Math|Vector|MakeVector _returnvalue_x _returnvalue_y _returnvalue_z))))
  (Pawn|Character|LaunchCharacterJump Player (Math|Vector|MakeVector _returnvalue_x_2 _returnvalue_y_2 (Math|Float|Clamp(Float) _returnvalue_z_2 (Variables|Default|GetMinZClamp) (Variables|Default|GetMaxZClamp))) true true true true)
  (Audio|PlaySoundatLocation "/PistolCanary_Event_Audio/Assets/Section02/Impacts/SC_PistolCanary_Section02_Tentacle_Portal_Knockback.SC_PistolCanary_Section02_Tentacle_Portal_Knockback" (Transformation|GetActorLocation Player)))

(event OnComponentEndOverlap(Sphere)
  (params (OverlappedComponent :type Primitive%20Component%20Object%20Reference) (OtherActor :type Actor%20Object%20Reference) (OtherComp :type Primitive%20Component%20Object%20Reference) (OtherBodyIndex :type Integer))
  (Utilities|Casting|CastToFortPlayerPawn OtherActor)
  (Utilities|Time|ClearandInvalidateTimerbyHandle (Variables|Default|GetTimer_TeleportCheck)))

(fn ConstructionScript)
