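; Server-side teleporter that moves players to child-component spawn points for the Pistol Canary event, triggered by a global sequence event.
; While calendar event PC_03 is active, the Aircraft phase starts multi-squad relevancy through a spawned mutator, if enabled.
; Teleporting ejects players from hiding props, then places pawns per relevancy bucket or recursively round-robin with jitter.
; Bots can be destroyed en route, and delay nodes aim to throttle server load.
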(event EventBeginPlay
  ;[0.4881 0.6875 0.6875 0.5] Start
  (Networking|SwitchHasAuthority
    (:Authority
      (Class|FortAsyncActionAuthorityMatchReady|MatchReadyAsync
        (:OnReady
          (MCPContext|InitialCalendarSyncCompleted
            (:OnSyncComplete
              (bind (_asynctaskproxy _payload) (GameplayEventRouter|StartListeningToStatefulEvent (GameplayEventRouter|GetGlobalEventRouter) "/Events_Shared/Gameplay/GlobalEvents/Event_RouterEvent_String.Event_RouterEvent_String")
                (:then
                  (bind _biseventactive (CalendarEvents|GetCalendarEventInformation "PC_03")) ; Start relevancy buckets if pre event is active.
                  (if _biseventactive
                    (bind _gamestate (Class|FortAsyncActionCurrentPlaylistReady|PlaylistDataReadyAsync)
                      (:OnReady
                        (Utilities|IsValid _gamestate
                          (:Is%20Valid
                            (bind _returnvalue (GameplayEventRouter|GetEventRouter "/Script/FortniteGame.FortGameState" _gamestate))
                            (bind (_asynctaskproxy_2 _payload_2) (GameplayEventRouter|StartListeningToStatefulEvent _returnvalue "/Script/FortniteGame.FortGamePhaseUpdatedEvent")
                              (:then
                                (Variables|Default|SetGamePhaseAsyncTask _asynctaskproxy_2))
                              (:OnEventReceived
                                (if (Utilities|Enum|Equal(Enum) (Utilities|Struct|BreakFortGamePhaseUpdatedEvent _payload_2) "Aircraft")
                                  (GameplayEventRouter|StopListeningtoStatefulEvent (Variables|Default|GetGamePhaseAsyncTask))
                                  (CallFunction|SpawnRelevanceMutator))))))))))
                (:OnEventReceived
                  (bind _payload_eventname_4_40c360d849aa0a7e7043d1be14586049 (Utilities|Struct|BreakEventRouterEventString _payload))
                  (if (Utilities|String|Equal,CaseInsensitive(String) _payload_eventname_4_40c360d849aa0a7e7043d1be14586049 (Variables|Default|GetSequenceEventName))
                    (CallFunction|TeleportAndActivateRelevance)))))))))))

(event Custom|Teleport%20and%20Activate%20Relevance
  (Networking|SwitchHasAuthority
    (:Authority
      (CallFunction|ForceHidingStop)
      (Variables|Default|SetSpawn%20Points (Components|GetChildrenComponents (Variables|Default|GetDefaultSceneRoot) true))
      (bind _outfortplayerpawns (Fortnite|GetFortPlayerPawns))
      (Variables|Default|SetPlayer%20Pawns _outfortplayerpawns)
      (Variables|Default|SetSpawnpoint%20Index 0) ; init to zero
      (Variables|Default|SetPlayer%20Index 0)
      (CallFunction|TeleportPlayers))))

(event Custom|PC_TeleportPlayers
  (Networking|SwitchHasAuthority
    (:Authority
      (CallFunction|ForceHidingStop)
      (Variables|Default|SetSpawn%20Points (Components|GetChildrenComponents (Variables|Default|GetDefaultSceneRoot) true))
      (bind _outfortplayerpawns (Fortnite|GetFortPlayerPawns))
      (Variables|Default|SetPlayer%20Pawns _outfortplayerpawns)
      (Variables|Default|SetSpawnpoint%20Index 0) ; init to zero
      (Variables|Default|SetPlayer%20Index 0)
      (CallFunction|TeleportPlayers))))

(event Custom|PC_ZoneDL
  (CallFunction|SpawnRelevanceMutator)
  (Development|PrintString "I have spawned relevance" true true "(R=0.000000,G=0.660000,B=1.000000,A=1.000000)" 3e+01)
  (Variables|Default|SetSpawn%20Points (Components|GetChildrenComponents (Variables|Default|GetDefaultSceneRoot) true)))

(event Custom|PC_StopHide
  (CallFunction|ForceHidingStop))

(event Custom|PC_Event
  (CallFunction|SpawnRelevanceMutator)
  (Development|PrintString "I have spawned relevance" true true "(R=0.000000,G=0.660000,B=1.000000,A=1.000000)" 3e+01)
  (Variables|Default|SetSpawn%20Points (Components|GetChildrenComponents (Variables|Default|GetDefaultSceneRoot) true)))

(event Custom|pistolcanary
  (CallFunction|SpawnRelevanceMutator)
  (Development|PrintString "I have spawned relevance" true true "(R=0.000000,G=0.660000,B=1.000000,A=1.000000)" 3e+01)
  (Variables|Default|SetSpawn%20Points (Components|GetChildrenComponents (Variables|Default|GetDefaultSceneRoot) true)))

(event Custom|CallNextTeleport
  ;[0.4359 0.431 0.6667 0.5] Teleport Players
  (Math|Integer|IncrementInt (Variables|Default|GetPlayer%20Index)
    (:%20%20
      (if (> (Variables|Default|GetSpawnpoint%20Index) (Utilities|Array|LastIndex (Variables|Default|GetSpawn%20Points)))
        (Variables|Default|SetSpawnpoint%20Index 0))
      (if (> (Variables|Default|GetPlayer%20Index) (Utilities|Array|LastIndex (Variables|Default|GetPlayer%20Pawns)))
        (Fortnite|FortShippingLog 0 "Pistol Canary - Done Teleporting Players!! (BP_PistolCanary_PlayerTeleporter)")
        (else
          (CallFunction|TeleportNextPlayer))))))

(event Custom|Teleport%20Next%20Player
  (bind (_returnvalue_location _returnvalue_rotation) (Math|Transform|BreakTransform (Transformation|GetWorldTransform (Utilities|Array|Get(acopy) (Variables|Default|GetSpawn%20Points) (Variables|Default|GetSpawnpoint%20Index)))))
  (bind _player_pawns (Variables|Default|GetPlayer%20Pawns))
  (bind _current_player (Variables|Default|GetCurrent%20Player))
  ;[0.4359 0.431 0.6667 0.5] Teleport Players
  (if (Utilities|Array|IsNotEmpty _player_pawns)
    (Variables|Default|SetCurrent%20Player (Utilities|Array|Get(acopy) _player_pawns (Variables|Default|GetPlayer%20Index)))
    (if (Pawn|IsPlayerControlled _current_player)
      (Transform|Actor|TeleportPlayerPawn (Variables|Default|GetCurrent%20Player) (Math|Vector|MakeVector (+ (.x _returnvalue_location) (Math|Random|RandomFloatinRange -25.0 25.0)) (+ (.y _returnvalue_location) (Math|Random|RandomFloatinRange -25.0 25.0)) (.z _returnvalue_location)) _returnvalue_rotation true true)
      (if (== (Math|Integer|%25%28Integer%29 (Variables|Default|GetSpawnpoint%20Index) 10) 0)
        (Utilities|FlowControl|Delay 0.1) ; every 10 players, pause for\nserver performance
        (else
          _))
      (Math|Integer|IncrementInt (Variables|Default|GetSpawnpoint%20Index)
        (:%20%20
          (Math|Integer|IncrementInt (Variables|Default|GetPlayer%20Index)
            (:%20%20
              (if (> (Variables|Default|GetSpawnpoint%20Index) (Utilities|Array|LastIndex (Variables|Default|GetSpawn%20Points)))
                (Variables|Default|SetSpawnpoint%20Index 0))
              (if (> (Variables|Default|GetPlayer%20Index) (Utilities|Array|LastIndex (Variables|Default|GetPlayer%20Pawns)))
                (Fortnite|FortShippingLog 0 "Pistol Canary - Done Teleporting Players!! (BP_PistolCanary_PlayerTeleporter)")
                (else
                  (CallFunction|TeleportNextPlayer)))))))
      (elif (GameplayTags|HasAnyMatchingGameplayTags _current_player (Variables|Default|GetHero%20Tag%20Container))
        (CallFunction|CallNextTeleport)
        (elif (ScalableFloat|GetValueasBool (Variables|Default|GetDestroy%20Bots%20On%20Teleport) 1.0)
          (Fortnite|FortShippingLog 0 (Utilities|String|Append "Pistol Canary - eliminating player bot while teleporting players to the event start (BP_PistolCanary_PlayerTeleporter)" ", " (Fortnite|GetPlayerNameSafe (Variables|Default|GetCurrent%20Player))))
          (Actor|DestroyActor (Variables|Default|GetCurrent%20Player)) ; eliminate the bot
          (CallFunction|CallNextTeleport)
          (else
            (CallFunction|CallNextTeleport)))))))

(event Custom|TeleportPlayers
  ;[0.4359 0.431 0.6667 0.5] Teleport Players
  (if (ScalableFloat|GetValueasBool (Variables|Default|GetUse%20Relevancy%20Buckets) 1.0)
    (bind _relevantpawns (Class|FortAthenaMutatorSpecialRelevancy|GetRelevantPawnGroups (Class|SpecialEventHelpers|GetSpecialRelevancyMutator)))
    (Fortnite|FortShippingLog 0 (Utilities|String|Append "PistolCanary PlayerTeleporter bucket pawns to teleport =" (Utilities|String|ToString(Integer) (Utilities|Array|Length _relevantpawns))))
    (for (_array_element _array_index) _relevantpawns
      (for (_array_element_2 _array_index_2) (Utilities|Struct|BreakRelevantPawnArray _array_element)
        (Variables|Default|SetCurrent%20Player _array_element_2)
        (if (Pawn|IsBotControlled (Variables|Default|GetCurrent%20Player))
          (Fortnite|FortShippingLog 0 "Pistol Canary - eliminating player bot while teleporting players to the event start (BP_PistolCanary_PlayerTeleporter)")
          (Actor|DestroyActor (Variables|Default|GetCurrent%20Player)) ; eliminate the bot
          (else
            (bind (_returnvalue_location _returnvalue_rotation) (Math|Transform|BreakTransform (Transformation|GetWorldTransform (Utilities|Array|Get(acopy) (Variables|Default|GetSpawn%20Points) _array_index_2))))
            (Transform|Actor|TeleportPlayerPawn (Variables|Default|GetCurrent%20Player) _returnvalue_location _returnvalue_rotation true true)
            (if (== (Math|Integer|%25%28Integer%29 _array_index_2 10) 0)
              (Utilities|FlowControl|Delay 0.1) ; every 10 players, pause for\nserver performance
            ))))
      (Development|PrintString (Utilities|String|Append "Bucket " (Utilities|String|ToString(Integer) _array_index) " has been teleported")))
    (Development|PrintString "All buckets teleported")
    (else
      (CallFunction|TeleportNextPlayer))))

(event Custom|Teleport%20Next%20Pawn
  (params (PlayerToTeleport :type Fort%20Player%20Pawn%20Object%20Reference))
  (bind (_returnvalue_location _returnvalue_rotation) (Math|Transform|BreakTransform (Transformation|GetWorldTransform (Utilities|Array|RandomArrayItem (Variables|Default|GetSpawn%20Points)))))
  (Networking|SwitchHasAuthority
    (:Authority
      (Utilities|IsValid PlayerToTeleport
        (:Is%20Valid
          (Transform|Actor|TeleportPlayerPawn PlayerToTeleport _returnvalue_location _returnvalue_rotation true true))))))

(fn ForceHidingStop
  ; Force hiding props in the map to eject players prior to teleport
  (bind _outactors (GameplayTags|GetAllActorsOfClassMatchingTagQuery "/Game/Athena/Items/EnvironmentalItems/HidingProps/B_HidingProp.B_HidingProp_C" (Variables|Default|GetHidingTag)))
  (Variables|Default|SetPropsHidingPlayers _outactors)
  (for _array_element (Variables|Default|GetPropsHidingPlayers)
    (if (> (Utilities|Array|Length (Variables|/Game/Athena/Items/EnvironmentalItems/HidingProps/B_HidingProp.B_HidingProp_C|GetHidingPlayers _array_element)) 0)
      (for _array_element_2 (Variables|/Game/Athena/Items/EnvironmentalItems/HidingProps/B_HidingProp.B_HidingProp_C|GetHidingPlayers _array_element)
        (Class|BHidingProp|StopHiding _array_element _array_element_2))))
  (CallFunction|TeleportPlayers))

(fn SpawnRelevanceMutator
  (if (ScalableFloat|GetValueasBool (Variables|Default|GetUse%20Relevancy%20Buckets) 1.0)
    (Utilities|IsValid (Variables|Default|GetRelevancyMutator)
      (:Is%20Valid
        _)
      (:Is%20Not%20Valid
        (bind _returnvalue (Class|SpecialEventGameplayLibrary|SpawnMutator "/PistolCanary/Gameplay/M_SE_PC_Relevancy.M_SE_PC_Relevancy_C"))
        (Variables|Default|SetRelevancyMutator _returnvalue)))
    (Mutator|SetMutatorActive (Variables|Default|GetRelevancyMutator) true)
    (Class|FortAthenaMutatorSpecialRelevancy|StartMultiSquadRelevancy (Class|SpecialEventHelpers|GetSpecialRelevancyMutator) (ScalableFloat|GetValueasIntegerwithContext (Variables|Default|GetMax%20Players%20Per%20Relevancy%20Group) 1.0))))

(fn ConstructionScript)
