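; BP_PistolCanary_PostEventLaunchPad is a launch pad Blueprint that flings overlapping Fortnite player pawns.
; The Sphere's begin-overlap and a custom actor-overlap handler, whose binding is not shown, each cast the other actor to FortPlayerPawn and call LaunchPlayerPawn.
; Both derive the launch from the Arrow component's world rotation multiplied by a fixed vector, differing only in one boolean launch flag.
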(event Custom|OnActorBeginOverlap_Event
  (params (OverlappedActor :type Actor%20Object%20Reference) (OtherActor :type Actor%20Object%20Reference))
  (bind _asfort_player_pawn (Utilities|Casting|CastToFortPlayerPawn OtherActor))
  (Utilities|LaunchPlayerPawn _asfort_player_pawn _asfort_player_pawn (Utilities|Operators|Multiply (Transformation|GetWorldRotation (Variables|Default|GetArrow)) (5e+03 5e+03 5e+03)) true true true))

(event OnComponentBeginOverlap(Sphere)
  (params (OverlappedComponent :type Primitive%20Component%20Object%20Reference) (OtherActor :type Actor%20Object%20Reference) (OtherComp :type Primitive%20Component%20Object%20Reference) (OtherBodyIndex :type Integer) (bFromSweep :type Boolean) (SweepResult :type Hit%20Result%20Structure%20%28by%20ref%29))
  (bind _asfort_player_pawn (Utilities|Casting|CastToFortPlayerPawn OtherActor))
  (Utilities|LaunchPlayerPawn _asfort_player_pawn _asfort_player_pawn (Utilities|Operators|Multiply (Transformation|GetWorldRotation (Variables|Default|GetArrow)) (5e+03 5e+03 5e+03)) false true true))

(fn ConstructionScript)
