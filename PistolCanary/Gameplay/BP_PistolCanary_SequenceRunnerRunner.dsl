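; Server-authoritative master runner for the PistolCanary live event, chaining its three sequence sections.
; After match-ready and calendar sync, calendar events preload Section 1 data layers, then start the event and destroy the countdown pre-roll runner.
; Starting sets special-event data, clears dangling replicated sequences and plays Section 1; each section's stop delegate starts the next.
; Section 3's end unloads data layers, broadcasts PC_EventEnded and clears special-event data.
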(event EventBeginPlay
  ;[0.1326 0.15 0.01508 0.5] Calendar
  (Networking|SwitchHasAuthority
    (:Authority
      (if (Variables|Default|GetFollowCalendarEvents)
        (Class|FortAsyncActionAuthorityMatchReady|MatchReadyAsync
          (:OnReady
            (MCPContext|InitialCalendarSyncCompleted)
            (Utilities|FlowControl|Sequence
              (:then_0
                (MCPContext|WatchCalendarEvent (Variables|Default|GetCalendarEventToWatch_PC_Preload)
                  (:OnEventActive
                    (bind _returnvalue (WorldPartition|GetDataLayerManager))
                    (DataLayers|SetDataLayerRuntimeState _returnvalue "/PistolCanary/DataLayers/DL_PistolCanary_S1.DL_PistolCanary_S1" "Loaded" false :!class /Script/Engine.DataLayerManager)
                    (DataLayers|SetDataLayerRuntimeState _returnvalue "/PistolCanary/DataLayers/DL_PistolCanary_S1_to_End_S3.DL_PistolCanary_S1_to_End_S3" "Loaded" false :!class /Script/Engine.DataLayerManager))) ; pre-load asset
              )
              (:then_1
                (MCPContext|WatchCalendarEvent (Variables|Default|GetCalendarEventToWatch_PC_Event)
                  (:OnEventActive
                    (Fortnite|FortShippingLog self "PC_Event Calendar Event is active")
                    (CallFunction|PCEvent 0.0)
                    (bind _returnvalue_2 (Actor|GetActorOfClass "/PistolCanary_Countdown/Gameplay/BP_PistolCanary_SequenceRunner_PreRollLayout.BP_PistolCanary_SequenceRunner_PreRollLayout_C"))
                    (Actor|DestroyActor _returnvalue_2))) ; pre-load asset
              ))))))))

(event Custom|pistolcanary
  (params (Percentage :type Float%20%28double-precision%29))
  (bind _section1 (Variables|Default|GetSection1))
  (if (Networking|IsServer)
    (CallFunction|SpecialEventBegin)
    (CallFunction|CleanUpDanglingSequences)
    (Default|BindEventtoOnSequenceStopped _section1 (delegate Custom|DISP_OnStopped_1))
    (Class|BPPistolCanarySequenceRunnerSection1|StartSection1 _section1 0.0)
    (GameplayEventRouter|BroadcastStatefulEvent (GameplayEventRouter|GetGlobalEventRouter) (Variables|Default|GetStartPCEvent) 0)
    (Utilities|FlowControl|Delay 4.4)
    (Development|ExecuteConsoleCommand "cheat ke * PC_TurnOffLights")))

(event Custom|PC_Event
  (params (Percentage :type Float%20%28double-precision%29))
  (bind _section1 (Variables|Default|GetSection1))
  (if (Networking|IsServer)
    (CallFunction|SpecialEventBegin)
    (CallFunction|CleanUpDanglingSequences)
    (Default|BindEventtoOnSequenceStopped _section1 (delegate Custom|DISP_OnStopped_1))
    (Class|BPPistolCanarySequenceRunnerSection1|StartSection1 _section1 0.0)
    (GameplayEventRouter|BroadcastStatefulEvent (GameplayEventRouter|GetGlobalEventRouter) (Variables|Default|GetStartPCEvent) 0)
    (Utilities|FlowControl|Delay 4.4)
    (Development|ExecuteConsoleCommand "cheat ke * PC_TurnOffLights")))

(event Custom|DISP_OnStopped_1
  (bind _section2 (Variables|Default|GetSection2))
  (Default|UnbindEventfromOnSequenceStopped (Variables|Default|GetSection1) (delegate Custom|DISP_OnStopped_1))
  (Development|PrintString "RUNNERrunner: end play 1")
  (Default|BindEventtoOnSequenceStopped _section2 (delegate Custom|DISP_OnStopped_2))
  (Class|BPPistolCanarySequenceRunnerSection2|StartSection2 _section2 0.0))

(event Custom|SpecialEventBegin
  (Class|FortKismetLibrary|SetSpecialEventData "PistolCanary"))

(event Custom|SpecialEventEnd
  (Class|FortKismetLibrary|ClearSpecialEventData))

(event EventEndPlay
  (params (EndPlayReason :type EEndPlayReason%20Enum :default "Destroyed"))
  (Class|FortKismetLibrary|ClearSpecialEventData))

(event Custom|DISP_OnStopped_2
  (bind _section3 (Variables|Default|GetSection3))
  (Default|UnbindEventfromOnSequenceStopped (Variables|Default|GetSection2) (delegate Custom|DISP_OnStopped_2))
  (Development|PrintString "RUNNERrunner: end play 2")
  (Default|BindEventtoOnSequenceStopped _section3 (delegate Custom|DISP_OnEndPlay_3))
  (Class|BPPistolCanarySequenceRunnerSection3|StartSection3 _section3 0.0))

(event Custom|PC_Inventory
  (GameplayEventRouter|BroadcastStatefulEvent (GameplayEventRouter|GetGlobalEventRouter) (Variables|Default|GetInventoryEvent) 0))

(event Custom|DisableZoneInventoryMod
  (GameplayEventRouter|BroadcastStatefulEvent (GameplayEventRouter|GetGlobalEventRouter) (Variables|Default|GetInventoryEvent) 0))

(event Custom|DISP_OnEndPlay_3
  (Default|UnbindEventfromOnSequenceStopped (Variables|Default|GetSection3) (delegate Custom|DISP_OnEndPlay_3))
  (Development|PrintString "RUNNERrunner: end play 3")
  (bind _returnvalue (WorldPartition|GetDataLayerManager))
  (DataLayers|SetDataLayerRuntimeState _returnvalue "/PistolCanary/DataLayers/DL_PistolCanary_S3.DL_PistolCanary_S3" "Unloaded" false :!class /Script/Engine.DataLayerManager)
  (DataLayers|SetDataLayerRuntimeState _returnvalue "/PistolCanary_Countdown/DataLayers/DL_PistolCanary_Day1_to_End_S3.DL_PistolCanary_Day1_to_End_S3" "Unloaded" false :!class /Script/Engine.DataLayerManager)
  (DataLayers|SetDataLayerRuntimeState _returnvalue "/PistolCanary_Countdown/DataLayers/DL_PistolCanary_S0_to_End_S3.DL_PistolCanary_S0_to_End_S3" "Unloaded" false :!class /Script/Engine.DataLayerManager)
  (DataLayers|SetDataLayerRuntimeState _returnvalue "/PistolCanary/DataLayers/DL_PistolCanary_S1_to_End_S3.DL_PistolCanary_S1_to_End_S3" "Unloaded" false :!class /Script/Engine.DataLayerManager)
  (DataLayers|SetDataLayerRuntimeState _returnvalue 0 "Unloaded" false :!class /Script/Engine.DataLayerManager)
  (DataLayers|SetDataLayerRuntimeState _returnvalue 0 "Unloaded" false :!class /Script/Engine.DataLayerManager)
  (GameplayEventRouter|BroadcastStatefulEvent (GameplayEventRouter|GetGlobalEventRouter) (Utilities|Struct|MakeEventRouterEventString "PC_EventEnded") 0)
  (bind _returnvalue_2 (Actor|GetActorOfClass "/PistolCanary_Countdown/Gameplay/BP_PistolCanary_MakeEnvironmentIndestructible.BP_PistolCanary_MakeEnvironmentIndestructible_C"))
  (Class|BPPistolCanaryMakeEnvironmentIndestructible|SetCanDestruct _returnvalue_2 false)
  (CallFunction|SpecialEventEnd))

(event Custom|TimeOfDaySetup
  (Development|ExecuteConsoleCommand "cheat settimeofday 2130")
  (Development|ExecuteConsoleCommand "cheat settimeofdayspeed 0"))

(event Custom|pc_teleport
  ; teleport
  (if (Networking|IsServer)
    (bind _outfortplayerpawns (Fortnite|GetFortPlayerPawns))
    (for _array_element _outfortplayerpawns
      (Transform|Actor|TeleportPawnNearLocation _array_element (Utilities|Operators|Add (Transformation|GetActorLocation) (0.0 0.0 175.0))))))

(fn CleanUpDanglingSequences
  (bind _outactors (Actor|GetAllActorsOfClass "/Script/LevelSequence.ReplicatedLevelSequenceActor"))
  (for _array_element _outactors
    (Actor|DestroyActor _array_element)))

(fn ConstructionScript)
