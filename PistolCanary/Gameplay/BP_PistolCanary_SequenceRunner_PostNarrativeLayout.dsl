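; Sequence runner for the PistolCanary event that sets the post-narrative map layout through World Partition data layers.
; On authority, after match-ready and initial calendar sync, two calendar events select the next-match or decayed-tentacle layout.
; Layout events unload event layers, swap castle states and toggle tentacle layers, spaced by delays.
; Cheat events preview each layout, including post-event, and StartPostNarrative pre-loads then plays the sequence on the server.
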(event EventBeginPlay
  ;[0.1326 0.15 0.01508 0.5] Calendar
  (Networking|SwitchHasAuthority
    (:Authority
      (if (Variables|Default|GetFollowCalendarEvents)
        (Class|FortAsyncActionAuthorityMatchReady|MatchReadyAsync
          (:OnReady
            (MCPContext|InitialCalendarSyncCompleted
              (:OnSyncComplete
                (Utilities|FlowControl|Sequence
                  (:then_0
                    (bind _biseventactive (CalendarEvents|GetCalendarEventInformation (Variables|Default|GetCalendarEventToWatch_PostEvent)))
                    (if _biseventactive
                      (CallFunction|LoadDataLayersforNextMatch)
                      (else
                        (Fortnite|FortShippingLog self "PostEvent Calendar Event is Inactive"))))
                  (:then_1
                    (bind _biseventactive_2 (CalendarEvents|GetCalendarEventInformation (Variables|Default|GetCalendarEventToWatch_TentacleStage2)))
                    (if _biseventactive_2
                      (CallFunction|LoadDataLayersforDecayedTentacle)
                      (else
                        (Fortnite|FortShippingLog self "DecayedTentacle Calendar Event is Inactive")))))))))))))

(event Custom|PC_PreviewDecayedTentacle
  ;[0 0 0] Console Command / Load and Play
  (Variables|Default|SetRunFromCheat true)
  (CallFunction|LoadDataLayersforDecayedTentacle))

(event Custom|PC_Tentacle2
  ;[0 0 0] Console Command / Load and Play
  (Variables|Default|SetRunFromCheat true)
  (CallFunction|LoadDataLayersforDecayedTentacle))

(event Custom|PC_PreviewNextMatch
  ;[0 0 0] Console Command / Load and Play
  (Variables|Default|SetRunFromCheat true)
  (CallFunction|LoadDataLayersforNextMatch))

(event Custom|PC_Next
  ;[0 0 0] Console Command / Load and Play
  (Variables|Default|SetRunFromCheat true)
  (CallFunction|LoadDataLayersforNextMatch))

(event Custom|PC_PreviewPostEvent
  ;[0 0 0] Console Command / Load and Play
  (Variables|Default|SetRunFromCheat true)
  (CallFunction|LoadDataLayersforPostEvent))

(event Custom|PC_After
  ;[0 0 0] Console Command / Load and Play
  (Variables|Default|SetRunFromCheat true)
  (CallFunction|LoadDataLayersforPostEvent))

(event Custom|StartPostNarrative
  (params (Percentage :type Float%20%28double-precision%29))
  ;[0 0 0] Console Command / Load and Play
  (Variables|Cheats|SetCheat%20Day%20Percentage Percentage) ; Allows skipping ahead in the\nsequence for testing. A value\nbetween 0 - 1.0
  (if (Networking|IsServer)
    (Variables|Cheats|SetStarting%20Method "NewEnumerator1") ; Keeps some calendar logic from running
    (Variables|Default|SetRunFromCheat true)
    (CallFunction|%28CallOnServer%29Pre-LoadSequence)))

(event EventSequenceStopped(replicated)
  ; EndSequence Callback for RunnerRunner
  (Default|CallOnSequenceStopped))

(event EventPre-LoadSequenceComplete
  ;[0 0 0] Console Command / Load and Play
  (|CallParentFunction)
  (if (Networking|IsServer)
    (if (Variables|Default|GetRunFromCheat)
      (CallFunction|LoadSequence) ; Play the sequence.
    )))

(event Custom|LoadDataLayersForPostEvent
  (CallFunction|UnloadPistolCanary)
  (Utilities|FlowControl|Delay 2.0)
  (CallFunction|SwitchtoPreEventCastle)
  (Utilities|FlowControl|Delay 3.0)
  (bind _returnvalue (WorldPartition|GetDataLayerManager))
  (DataLayers|SetDataLayerRuntimeState _returnvalue "/PistolCanary_Countdown/DataLayers/DL_PistolCanary_Day1_to_MatchEnd.DL_PistolCanary_Day1_to_MatchEnd" "Activated" false :!class /Script/Engine.DataLayerManager)
  (DataLayers|SetDataLayerRuntimeState _returnvalue "/PistolCanary_Countdown/DataLayers/DL_PistolCanary_Day2_to_MatchEnd.DL_PistolCanary_Day2_to_MatchEnd" "Activated" false :!class /Script/Engine.DataLayerManager)
  (DataLayers|SetDataLayerRuntimeState _returnvalue "/PistolCanary/DataLayers/DL_PistolCanary_S2_to_MatchEnd.DL_PistolCanary_S2_to_MatchEnd" "Activated" false :!class /Script/Engine.DataLayerManager)
  (DataLayers|SetDataLayerRuntimeState _returnvalue "/PistolCanary/DataLayers/DL_PistolCanary_S3_to_MatchEnd.DL_PistolCanary_S3_to_MatchEnd" "Activated" false :!class /Script/Engine.DataLayerManager)
  (DataLayers|SetDataLayerRuntimeState _returnvalue "/PistolCanary/DataLayers/DL_PistolCanary_S3_to_PostEvent_SeveredTentacle.DL_PistolCanary_S3_to_PostEvent_SeveredTentacle" "Activated" false :!class /Script/Engine.DataLayerManager)
  (DataLayers|SetDataLayerRuntimeState _returnvalue 0 "Unloaded" false :!class /Script/Engine.DataLayerManager)
  (DataLayers|SetDataLayerRuntimeState _returnvalue 0 "Unloaded" false :!class /Script/Engine.DataLayerManager)
  (DataLayers|SetDataLayerRuntimeState _returnvalue 0 "Unloaded" false :!class /Script/Engine.DataLayerManager)
  (DataLayers|SetDataLayerRuntimeState _returnvalue 0 "Unloaded" false :!class /Script/Engine.DataLayerManager)
  (DataLayers|SetDataLayerRuntimeState _returnvalue 0 "Unloaded" false :!class /Script/Engine.DataLayerManager)
  (DataLayers|SetDataLayerRuntimeState _returnvalue "/PistolCanary/DataLayers/DL_PistolCanary_PostEvent_DecayedTentacle.DL_PistolCanary_PostEvent_DecayedTentacle" "Unloaded" false :!class /Script/Engine.DataLayerManager)
  (DataLayers|SetDataLayerRuntimeState _returnvalue "/PistolCanary_Countdown/DataLayers/DL_PistolCanary_InnerSpire1.DL_PistolCanary_InnerSpire1" "Unloaded" false :!class /Script/Engine.DataLayerManager)
  (DataLayers|SetDataLayerRuntimeState _returnvalue "/PistolCanary_Countdown/DataLayers/DL_PistolCanary_InnerSpire2.DL_PistolCanary_InnerSpire2" "Unloaded" false :!class /Script/Engine.DataLayerManager)
  (DataLayers|SetDataLayerRuntimeState _returnvalue "/PistolCanary_Countdown/DataLayers/DL_PistolCanary_InnerSpire3.DL_PistolCanary_InnerSpire3" "Unloaded" false :!class /Script/Engine.DataLayerManager)
  (DataLayers|SetDataLayerRuntimeState _returnvalue "/PistolCanary_Countdown/DataLayers/DL_PistolCanary_InnerSpire4.DL_PistolCanary_InnerSpire4" "Unloaded" false :!class /Script/Engine.DataLayerManager))

(event Custom|UnloadPistolCanary
  (bind _returnvalue (WorldPartition|GetDataLayerManager))
  (DataLayers|SetDataLayerRuntimeState _returnvalue "/PistolCanary_Countdown/DataLayers/EDL_PistolCanary_Countdown.EDL_PistolCanary_Countdown" "Unloaded" true :!class /Script/Engine.DataLayerManager)
  (DataLayers|SetDataLayerRuntimeState _returnvalue "/PistolCanary/DataLayers/DL_PistolCanary_EventParent.DL_PistolCanary_EventParent" "Unloaded" true :!class /Script/Engine.DataLayerManager)
  (Utilities|FlowControl|Delay 0.2)
  (DataLayers|SetDataLayerRuntimeState _returnvalue "/PistolCanary/DataLayers/DL_PistolCanary_EventParent.DL_PistolCanary_EventParent" "Activated" false :!class /Script/Engine.DataLayerManager))

(event Custom|LoadDataLayersForDecayedTentacle
  (CallFunction|LoadDataLayersforNextMatch)
  (Utilities|FlowControl|Delay 2.0)
  (bind _returnvalue (WorldPartition|GetDataLayerManager))
  (DataLayers|SetDataLayerRuntimeState _returnvalue "/PistolCanary/DataLayers/DL_PistolCanary_PostEvent_DecayedTentacle.DL_PistolCanary_PostEvent_DecayedTentacle" "Activated" false :!class /Script/Engine.DataLayerManager)
  (DataLayers|SetDataLayerRuntimeState _returnvalue 0 "Unloaded" false :!class /Script/Engine.DataLayerManager)
  (DataLayers|SetDataLayerRuntimeState _returnvalue "/PistolCanary/DataLayers/DL_PistolCanary_S3_to_PostEvent_SeveredTentacle.DL_PistolCanary_S3_to_PostEvent_SeveredTentacle" "Unloaded" false :!class /Script/Engine.DataLayerManager))

(event Custom|LoadDataLayersForNextMatch
  (CallFunction|UnloadPistolCanary)
  (CallFunction|SwitchtoDestroyedCastle)
  (bind _returnvalue (WorldPartition|GetDataLayerManager))
  (DataLayers|SetDataLayerRuntimeState _returnvalue "/PistolCanary/DataLayers/DL_PistolCanary_S3_to_PostEvent_SeveredTentacle.DL_PistolCanary_S3_to_PostEvent_SeveredTentacle" "Activated" false :!class /Script/Engine.DataLayerManager))

(event Custom|SwitchToDestroyedCastle
  (bind _returnvalue (WorldPartition|GetDataLayerManager))
  (DataLayers|SetDataLayerRuntimeState _returnvalue 0 "Unloaded" false :!class /Script/Engine.DataLayerManager)
  (DataLayers|SetDataLayerRuntimeState _returnvalue 0 "Unloaded" false :!class /Script/Engine.DataLayerManager)
  (DataLayers|SetDataLayerRuntimeState _returnvalue 0 "Unloaded" false :!class /Script/Engine.DataLayerManager)
  (DataLayers|SetDataLayerRuntimeState _returnvalue 0 "Unloaded" false :!class /Script/Engine.DataLayerManager)
  (DataLayers|SetDataLayerRuntimeState _returnvalue 0 "Unloaded" false :!class /Script/Engine.DataLayerManager)
  (Utilities|FlowControl|Delay 1.0)
  (DataLayers|SetDataLayerRuntimeState _returnvalue 0 "Activated" false :!class /Script/Engine.DataLayerManager)
  (DataLayers|SetDataLayerRuntimeState _returnvalue 0 "Activated" false :!class /Script/Engine.DataLayerManager))

(event Custom|SwitchToPreEventCastle
  (bind _returnvalue (WorldPartition|GetDataLayerManager))
  (DataLayers|SetDataLayerRuntimeState _returnvalue 0 "Unloaded" false :!class /Script/Engine.DataLayerManager)
  (DataLayers|SetDataLayerRuntimeState _returnvalue 0 "Unloaded" false :!class /Script/Engine.DataLayerManager)
  (DataLayers|SetDataLayerRuntimeState _returnvalue 0 "Unloaded" false :!class /Script/Engine.DataLayerManager)
  (DataLayers|SetDataLayerRuntimeState _returnvalue 0 "Unloaded" false :!class /Script/Engine.DataLayerManager)
  (Utilities|FlowControl|Delay 1.0)
  (DataLayers|SetDataLayerRuntimeState _returnvalue 0 "Activated" false :!class /Script/Engine.DataLayerManager)
  (DataLayers|SetDataLayerRuntimeState _returnvalue 0 "Activated" false :!class /Script/Engine.DataLayerManager)
  (DataLayers|SetDataLayerRuntimeState _returnvalue 0 "Activated" false :!class /Script/Engine.DataLayerManager)
  (DataLayers|SetDataLayerRuntimeState _returnvalue 0 "Activated" false :!class /Script/Engine.DataLayerManager)
  (DataLayers|SetDataLayerRuntimeState _returnvalue 0 "Activated" false :!class /Script/Engine.DataLayerManager)
  (DataLayers|SetDataLayerRuntimeState _returnvalue 0 "Activated" false :!class /Script/Engine.DataLayerManager))

(fn ConstructionScript
  (|CallParentFunction))
