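; Sequence runner for Section 1 of the PistolCanary event, with cheat entry points to play it from a chosen point.
; PC_Section1 and StartSection1 store a skip-ahead percentage, then on the server flag a cheat start, pre-load the sequence and broadcast a verb hiding the pre-event training urgent quest.
; Pre-load completion loads and plays the sequence for cheat runs.
; It also forwards the replicated stop event and triggers a tentacle camera shake on player controller 0.
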(event Custom|PC_Section1
  (params (Cheat%20Day%20Percentage :type Float%20%28double-precision%29))
  ;[0 0 0] Console Command / Load and Play
  (Variables|Cheats|SetCheat%20Day%20Percentage Cheat%20Day%20Percentage) ; Allows skipping ahead in the\nsequence for testing. A value\nbetween 0 - 1.0
  (if (Networking|IsServer)
    (Variables|Cheats|SetStarting%20Method "NewEnumerator1") ; Keeps some calendar logic from running
    (Variables|Default|SetRun%20From%20Cheat true)
    (CallFunction|%28CallOnServer%29Pre-LoadSequence)
    (CallFunction|StopTrainingUrgentQuest)))

(event Custom|Stop%20Training%20Urgent%20Quest
  ;[1 0.8713 0.9167] Stop Training Urgent Quest
  (bind _outfortplayerpawns (Fortnite|GetFortPlayerPawns))
  (for _array_element _outfortplayerpawns
    (CustomVerb|BroadcastCustomVerbMessage (CustomVerb|MakeCustomVerbMessage _array_element "(TagName=\"CustomVerb.PistolCanary.HideUrgentQuest\")")) ; stop urgent quest for earning training \npoints during the pre-event
  ))

(event Custom|StartSection1
  (params (Percentage :type Float%20%28double-precision%29))
  ;[0 0 0] Console Command / Load and Play
  (Variables|Cheats|SetCheat%20Day%20Percentage Percentage) ; Allows skipping ahead in the\nsequence for testing. A value\nbetween 0 - 1.0
  (if (Networking|IsServer)
    (Variables|Cheats|SetStarting%20Method "NewEnumerator1") ; Keeps some calendar logic from running
    (Variables|Default|SetRun%20From%20Cheat true)
    (CallFunction|%28CallOnServer%29Pre-LoadSequence)
    (CallFunction|StopTrainingUrgentQuest)))

(event EventPre-LoadSequenceComplete
  ;[0 0 0] Console Command / Load and Play
  (|CallParentFunction)
  (if (Networking|IsServer)
    (if (Variables|Default|GetRun%20From%20Cheat)
      (CallFunction|LoadSequence) ; Play the sequence.
    )))

(event EventSequenceStopped(replicated)
  ;[0 0 0] Console Command / Load and Play
  (Default|CallOnSequenceStopped))

(event Custom|CameraShakeTentacle
  (Game|Feedback|ClientStartCameraShake :self (Game|GetPlayerController 0) :Shake "/PistolCanary/Layout/BP/CameraShake/PistolCanary_SingleTentacleEmerge.PistolCanary_SingleTentacleEmerge_C" :Scale 1.0 :PlaySpace "CameraLocal"))

(fn ConstructionScript
  (|CallParentFunction))
