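; Sequence runner Blueprint for Section 2 of PistolCanary, with console-command events to load and play its sequence for testing.
; Two identical start events set a skip-ahead percentage and, on the server, flag a cheat start and pre-load; pre-load completion then plays the sequence.
; A separate event activates the S2-to-MatchEnd data layer.
; A helper picks vertical launch velocity from pawn height, aiming near Z 7600.
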(event Custom|PC_Section2
  (params (Cheat%20Day%20Percentage :type Float%20%28double-precision%29))
  ;[0 0 0] Console Command / Load and Play
  (Variables|Cheats|SetCheat%20Day%20Percentage Cheat%20Day%20Percentage) ; Allows skipping ahead in the\nsequence for testing. A value\nbetween 0 - 1.0
  (if (Networking|IsServer)
    (Variables|Cheats|SetStarting%20Method "NewEnumerator1") ; Keeps some calendar logic from running
    (Variables|Default|SetRun%20From%20Cheat true)
    (CallFunction|%28CallOnServer%29Pre-LoadSequence)
    (bind _returnvalue (WorldPartition|GetDataLayerManager))
    (DataLayers|SetDataLayerRuntimeState _returnvalue 0 "Unloaded" false :!class /Script/Engine.DataLayerManager)))

(event Custom|StartSection2
  (params (Percentage :type Float%20%28double-precision%29))
  ;[0 0 0] Console Command / Load and Play
  (Variables|Cheats|SetCheat%20Day%20Percentage Percentage) ; Allows skipping ahead in the\nsequence for testing. A value\nbetween 0 - 1.0
  (if (Networking|IsServer)
    (Variables|Cheats|SetStarting%20Method "NewEnumerator1") ; Keeps some calendar logic from running
    (Variables|Default|SetRun%20From%20Cheat true)
    (CallFunction|%28CallOnServer%29Pre-LoadSequence)
    (bind _returnvalue (WorldPartition|GetDataLayerManager))
    (DataLayers|SetDataLayerRuntimeState _returnvalue 0 "Unloaded" false :!class /Script/Engine.DataLayerManager)))

(event EventPre-LoadSequenceComplete
  ;[0 0 0] Console Command / Load and Play
  (|CallParentFunction)
  (if (Networking|IsServer)
    (if (Variables|Default|GetRun%20From%20Cheat)
      (CallFunction|LoadSequence) ; Play the sequence.
    )))

(event EventSequenceStopped(replicated)
  ;[0 0 0] Console Command / Load and Play
  (Default|CallOnSequenceStopped))

(event Custom|PC_DL_AI
  ;[0 0 0] Console Command / Load and Play
  (bind _returnvalue (WorldPartition|GetDataLayerManager))
  (DataLayers|SetDataLayerRuntimeState _returnvalue "/PistolCanary/DataLayers/DL_PistolCanary_S2_to_MatchEnd.DL_PistolCanary_S2_to_MatchEnd" "Activated" false :!class /Script/Engine.DataLayerManager))

(fn CheckZBeforeLaunch
  (params (currentPlayer :type Fort%20Player%20Pawn%20Object%20Reference))
  (bind (_returnvalue_x _returnvalue_y _returnvalue_z) (Math|Vector|BreakVector (Transformation|GetActorLocation currentPlayer)))
  ; Launches players to ~z:7600
  (switch Utilities|FlowControl|Switch|SwitchByRange _returnvalue_z
    (:MatchesRange_0
      (LocalVariables|SetLaunchZ 5e+03))
    (:MatchesRange_1
      (LocalVariables|SetLaunchZ 4.25e+03))
    (:MatchesRange_2
      (LocalVariables|SetLaunchZ 3.75e+03)))
  (return 0.0 0.0 (LocalVariables|GetLaunchZ)))

(fn ConstructionScript
  (|CallParentFunction))
