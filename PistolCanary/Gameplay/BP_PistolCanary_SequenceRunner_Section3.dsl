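; Sequence runner for Section 3 of the PistolCanary event, with cheat events that preload and play the sequence from a given percentage on the server.
; Custom events drive tentacle camera shakes, player hover via gameplay effects and event-router broadcasts, and timeline-based global slow motion.
; A client-only Superman moment snaps a tagged sequence beside the local pawn, with movement paused for a hardcoded delay.
; Helpers apply or remove effects on all players and pick launch Z by height.
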(event Custom|PC_Section3
  (params (Cheat%20Day%20Percentage :type Float%20%28double-precision%29))
  ;[0.3368 0.2862 0.4688] Console Command / Load and Play
  (Variables|Cheats|SetCheat%20Day%20Percentage Cheat%20Day%20Percentage) ; Allows skipping ahead in the\nsequence for testing. A value\nbetween 0 - 1.0
  (if (Networking|IsServer)
    (Variables|Cheats|SetStarting%20Method "NewEnumerator1") ; Keeps some calendar logic from running
    (Variables|Default|SetRun%20From%20Cheat true)
    (CallFunction|%28CallOnServer%29Pre-LoadSequence)
    (bind _returnvalue (WorldPartition|GetDataLayerManager))
    (DataLayers|SetDataLayerRuntimeState _returnvalue 0 "Unloaded" false :!class /Script/Engine.DataLayerManager)))

(event Custom|CameraShakeTripleTentacle
  ;[0.4169 0.5052 0.3673 0.5] Camera Shake
  (Game|Feedback|ClientStartCameraShake :self (Game|GetPlayerController 0) :Shake "/PistolCanary/Layout/BP/CameraShake/PistolCanary_TripleTentacleEmerge.PistolCanary_TripleTentacleEmerge_C" :Scale 1.0 :PlaySpace "CameraLocal"))

(event Custom|PC_3
  (params (Percentage :type Float%20%28double-precision%29))
  ;[0.3368 0.2862 0.4688] Console Command / Load and Play
  (Variables|Cheats|SetCheat%20Day%20Percentage Percentage) ; Allows skipping ahead in the\nsequence for testing. A value\nbetween 0 - 1.0
  (if (Networking|IsServer)
    (Variables|Cheats|SetStarting%20Method "NewEnumerator1") ; Keeps some calendar logic from running
    (Variables|Default|SetRun%20From%20Cheat true)
    (CallFunction|%28CallOnServer%29Pre-LoadSequence)
    (bind _returnvalue (WorldPartition|GetDataLayerManager))
    (DataLayers|SetDataLayerRuntimeState _returnvalue 0 "Unloaded" false :!class /Script/Engine.DataLayerManager)))

(event Custom|StartSection3
  (params (Percentage :type Float%20%28double-precision%29))
  ;[0.3368 0.2862 0.4688] Console Command / Load and Play
  (Variables|Cheats|SetCheat%20Day%20Percentage Percentage) ; Allows skipping ahead in the\nsequence for testing. A value\nbetween 0 - 1.0
  (if (Networking|IsServer)
    (Variables|Cheats|SetStarting%20Method "NewEnumerator1") ; Keeps some calendar logic from running
    (Variables|Default|SetRun%20From%20Cheat true)
    (CallFunction|%28CallOnServer%29Pre-LoadSequence)
    (bind _returnvalue (WorldPartition|GetDataLayerManager))
    (DataLayers|SetDataLayerRuntimeState _returnvalue 0 "Unloaded" false :!class /Script/Engine.DataLayerManager)))

(event Custom|CameraShakeTentacle
  ;[0.4169 0.5052 0.3673 0.5] Camera Shake
  (Game|Feedback|ClientStartCameraShake :self (Game|GetPlayerController 0) :Shake "/PistolCanary/Layout/BP/CameraShake/PistolCanary_SingleTentacleEmerge.PistolCanary_SingleTentacleEmerge_C" :Scale 1.0 :PlaySpace "CameraLocal"))

(event Custom|CameraShakeGravityAttack
  ;[0.4169 0.5052 0.3673 0.5] Camera Shake
  (Game|Feedback|ClientStartCameraShake :self (Game|GetPlayerController 0) :Shake "/PistolCanary/Layout/BP/CameraShake/PistolCanary_TentacleGravityAttack.PistolCanary_TentacleGravityAttack_C" :Scale 1.0 :PlaySpace "CameraLocal"))

(event EventPre-LoadSequenceComplete
  ;[0.3368 0.2862 0.4688] Console Command / Load and Play
  (|CallParentFunction)
  (if (Networking|IsServer)
    (if (Variables|Default|GetRun%20From%20Cheat)
      (CallFunction|LoadSequence) ; Play the sequence.
    )))

(event EventSequenceStopped(replicated)
  ;[0.3368 0.2862 0.4688] Console Command / Load and Play
  (Default|CallOnSequenceStopped))

(event Custom|PistolCanary_RemoveFreezePlayer
  ;[0.2749 0.4292 0.9167 0.5] Player Hovers during Triple Tentacle
  (GameplayEventRouter|BroadcastStatefulEvent (GameplayEventRouter|GetGlobalEventRouter) (Utilities|Struct|MakePistolCanaryEventPlayerHover "NewEnumerator1") (GameplayEventRouter|GetGlobalEventRouter)))

(event Custom|PistolCanary_LaunchPlayer
  (params (GameplayEffectClass :type Gameplay%20Effect%20Class%20Reference))
  ;[0.2749 0.4292 0.9167 0.5] Player Hovers during Triple Tentacle
  (Networking|SwitchHasAuthority
    (:Authority
      (bind _outfortplayerpawns (Fortnite|GetFortPlayerPawns))
      (for _array_element _outfortplayerpawns
        (GameplayEffects|ApplyGameplayEffectToTarget :self (Ability|GetFortAbilitySystemComponent _array_element) :GameplayEffectClass GameplayEffectClass :Target (Ability|GetFortAbilitySystemComponent _array_element) :Level 0.0) ; Make the player float up into the air\nas the triple tentacle builds energy
        (GameplayEffects|RemoveActiveEffectswithSourceTags (Ability|GetFortAbilitySystemComponent _array_element) "(GameplayTags=((TagName=\"Gameplay.Mod.FireRate\")))")))))

(event Custom|PC_StopHover
  ;[0.2749 0.4292 0.9167 0.5] Player Hovers during Triple Tentacle
  (GameplayEventRouter|BroadcastStatefulEvent (GameplayEventRouter|GetGlobalEventRouter) (Utilities|Struct|MakePistolCanaryEventPlayerHover "NewEnumerator1") (GameplayEventRouter|GetGlobalEventRouter)))

(event Custom|PC%20-%20Slo%20Mo
  (params (Start%20Value :type Float%20%28double-precision%29) (End%20Value :type Float%20%28double-precision%29) (Lerp%20Time :type Float%20%28double-precision%29))
  ;[0.5052 0.3536 0.4468 0.5] Slo Mo
  (Utilities|FlowControl|Sequence
    (:then_0
      (bind (_direction _alpha) (Timeline|Timeline%20-%20slo%20mo Stop)
        (:Update
          (Utilities|Time|SetGlobalTimeDilation (Math|Float|Lerp Start%20Value End%20Value _alpha)))))
    (:then_1
      (if (Networking|IsServer)
        (Components|Timeline|SetPlayRate (Variables|BP_PistolCanary_SequenceRunner_Section3|GetTimeline%20-%20slo%20mo) (/ 1.0 Lerp%20Time)) ; scale the timeline to play the \ncorrect length of time
        (Timeline|Timeline%20-%20slo%20mo PlayFromStart)))))

(event Custom|PC_Hover
  ;[0.2749 0.4292 0.9167 0.5] Player Hovers during Triple Tentacle
  (Networking|SwitchHasAuthority
    (:Authority
      (bind _outfortplayerpawns (Fortnite|GetFortPlayerPawns))
      (for _array_element _outfortplayerpawns
        (GameplayEffects|ApplyGameplayEffectToTarget :self (Ability|GetFortAbilitySystemComponent _array_element) :GameplayEffectClass (stale Custom|PistolCanary_LaunchPlayer GameplayEffectClass) :Target (Ability|GetFortAbilitySystemComponent _array_element) :Level 0.0) ; Make the player float up into the air\nas the triple tentacle builds energy
        (GameplayEffects|RemoveActiveEffectswithSourceTags (Ability|GetFortAbilitySystemComponent _array_element) "(GameplayTags=((TagName=\"Gameplay.Mod.FireRate\")))")))))

(event Custom|PC_TestTeleportFX
  ;[0.2749 0.4292 0.9167 0.5] Player Hovers during Triple Tentacle
  (Ability|SendGameplayEventtoActor (Game|Player|GetPlayerPawn (Fortnite|GetFirstLocalFortPlayerController true)) "(TagName=\"Abilities.Generic.Rift.Trigger\")") ; test teleport fx
)

(event Custom|Player%20Client%20Only%20Moment
  ;[0.2806 0.5145 0.5573 0.5] Superman Client Only / Localhost moment
  (if (Networking|IsServer)
    (GameplayEventRouter|BroadcastStatefulEvent (GameplayEventRouter|GetGlobalEventRouter) (Utilities|Struct|MakePistolCanaryEventPlayerHover "NewEnumerator4") 0) ; pause player horizontal movement\n..calls to GA_PistolCanary_PlayerHover
    (CallFunction|SupermanClientOnlyMoment(Client))
    (Utilities|FlowControl|Delay 6.0) ; Superman client only moment \nlasts about 6 seconds. Hacky to not trigger\nthis from something like "sequence ended,"\nbut the client plays the sequence and not the\nserver so we actually have no idea when it\nends on the server from watching the sequence.
    (GameplayEventRouter|BroadcastStatefulEvent (GameplayEventRouter|GetGlobalEventRouter) (Utilities|Struct|MakePistolCanaryEventPlayerHover "NewEnumerator5") 0) ; allow player horizontal movement again\n..calls to GA_PistolCanary_PlayerHover
  ))

(event Custom|PC_Localhost
  ;[0.2806 0.5145 0.5573 0.5] Superman Client Only / Localhost moment
  (bind _returnvalue (WorldPartition|GetDataLayerManager))
  (DataLayers|SetDataLayerRuntimeState _returnvalue "/PistolCanary/DataLayers/DL_PistolCanary_S3.DL_PistolCanary_S3" "Activated" false :!class /Script/Engine.DataLayerManager)
  (Utilities|FlowControl|Delay 1.0) ; let the data layer load
  (GameplayEventRouter|BroadcastStatefulEvent (GameplayEventRouter|GetGlobalEventRouter) (Utilities|Struct|MakePistolCanaryEventPlayerHover "NewEnumerator4") 0) ; pause player horizontal movement\n..calls to GA_PistolCanary_PlayerHover
  (CallFunction|SupermanClientOnlyMoment(Client))
  (Utilities|FlowControl|Delay 6.0) ; Superman client only moment \nlasts about 6 seconds. Hacky to not trigger\nthis from something like "sequence ended,"\nbut the client plays the sequence and not the\nserver so we actually have no idea when it\nends on the server from watching the sequence.
  (GameplayEventRouter|BroadcastStatefulEvent (GameplayEventRouter|GetGlobalEventRouter) (Utilities|Struct|MakePistolCanaryEventPlayerHover "NewEnumerator5") 0) ; allow player horizontal movement again\n..calls to GA_PistolCanary_PlayerHover
)

(event Custom|Superman%20Client%20Only%20Moment%20%28client%29
  (bind (_returnvalue_roll _returnvalue_pitch _returnvalue_yaw) (Math|Rotator|BreakRotator (Transformation|GetActorRotation (Game|Player|GetPlayerPawn (Fortnite|GetFirstLocalFortPlayerController true)))))
  ;[0.2806 0.5145 0.5573 0.5] Superman Client Only / Localhost moment
  (if (not (Networking|IsServer))
    (bind _outactors (Actor|GetAllActorsOfClasswithTag "/Script/LevelSequence.LevelSequenceActor" "Event.Sequence.Tag.01"))
    (if (Utilities|Array|IsNotEmpty _outactors)
      (Development|PrintString "Pistol Canary - teleport Superman to Player")
      (Transformation|SetActorLocationAndRotation (Utilities|Array|Get(acopy) _outactors) (+ (Math|Vector|MakeVector (.x (Transformation|GetActorLocation (Game|Player|GetPlayerPawn (Fortnite|GetFirstLocalFortPlayerController true)))) (.y (Transformation|GetActorLocation (Game|Player|GetPlayerPawn (Fortnite|GetFirstLocalFortPlayerController true)))) (+ (.z (Transformation|GetActorLocation (Game|Player|GetPlayerPawn (Fortnite|GetFirstLocalFortPlayerController true)))) (ScalableFloat|GetValueatLevel (Variables|Default|GetClient%20Only%20Moment%20-%20Z%20Offset) 1.0))) (* (Transformation|GetActorRightVector (Game|Player|GetPlayerPawn (Fortnite|GetFirstLocalFortPlayerController true))) (ScalableFloat|GetValueatLevel (Variables|Default|GetClient%20Only%20Moment%20-%20Right%20Offset) 1.0))) (Math|Rotator|MakeRotator _returnvalue_roll _returnvalue_pitch (- _returnvalue_yaw 9e+01)) false true)
      (Sequencer|Player|SetPlaybackPosition (Variables|/Script/LevelSequence.LevelSequenceActor|GetSequencePlayer (Utilities|Array|Get(acopy) _outactors)) (struct :Frame "(FrameNumber=(Value=0),SubFrame=0.000000)" :Time 6e+01 :MarkedFrame "" :Timecode "(Hours=0,Minutes=0,Seconds=0,Frames=0,Subframe=0.000000,bDropFrameFormat=False)" :PositionType "Time" :UpdateMethod "Jump" :bHasJumped true))
      (Sequencer|Player|Play (Variables|/Script/LevelSequence.LevelSequenceActor|GetSequencePlayer (Utilities|Array|Get(acopy) _outactors)) :!class /Script/MovieScene.MovieSceneSequencePlayer)
      (Utilities|FlowControl|Delay 2.0) ; just for breakpoints and debug, not needed
      (Development|PrintString "Pistol Canary - client only superman moment mid-moment")
      (else
        (Development|PrintString "Pistol Canary - failed to find client only superman sequence")))))

(event Custom|PC_ClientOnlyMoment
  ;[0.2806 0.5145 0.5573 0.5] Superman Client Only / Localhost moment
  (bind _returnvalue (WorldPartition|GetDataLayerManager))
  (DataLayers|SetDataLayerRuntimeState _returnvalue "/PistolCanary/DataLayers/DL_PistolCanary_S3.DL_PistolCanary_S3" "Activated" false :!class /Script/Engine.DataLayerManager)
  (Utilities|FlowControl|Delay 1.0) ; let the data layer load
  (GameplayEventRouter|BroadcastStatefulEvent (GameplayEventRouter|GetGlobalEventRouter) (Utilities|Struct|MakePistolCanaryEventPlayerHover "NewEnumerator4") 0) ; pause player horizontal movement\n..calls to GA_PistolCanary_PlayerHover
  (CallFunction|SupermanClientOnlyMoment(Client))
  (Utilities|FlowControl|Delay 6.0) ; Superman client only moment \nlasts about 6 seconds. Hacky to not trigger\nthis from something like "sequence ended,"\nbut the client plays the sequence and not the\nserver so we actually have no idea when it\nends on the server from watching the sequence.
  (GameplayEventRouter|BroadcastStatefulEvent (GameplayEventRouter|GetGlobalEventRouter) (Utilities|Struct|MakePistolCanaryEventPlayerHover "NewEnumerator5") 0) ; allow player horizontal movement again\n..calls to GA_PistolCanary_PlayerHover
)

(fn ApplyGameplayEffectToAllPlayers
  (params (NewGE :type Gameplay%20Effect%20Class%20Reference))
  (LocalVariables|SetGE_LocalVar NewGE)
  (if (Networking|IsDedicatedServer)
    (bind _outfortplayerpawns (Fortnite|GetFortPlayerPawns))
    (for _array_element _outfortplayerpawns
      (GameplayEffects|ApplyGameplayEffectToSelf :self (Ability|GetAbilitySystemComponent _array_element) :GameplayEffectClass (LocalVariables|GetGE_LocalVar) :Level 0.0))
    (Development|PrintString (Utilities|String|Append "Added " (Utilities|GetClassDisplayName (LocalVariables|GetGE_LocalVar)) " to all players"))))

(fn CheckZBeforeLaunch
  (params (currentPlayer :type Fort%20Player%20Pawn%20Object%20Reference))
  (bind (_returnvalue_x _returnvalue_y _returnvalue_z) (Math|Vector|BreakVector (Transformation|GetActorLocation currentPlayer)))
  ; Launches players to ~z:7600
  (switch Utilities|FlowControl|Switch|SwitchByRange _returnvalue_z
    (:MatchesRange_0
      (LocalVariables|SetLaunchZ 5e+03))
    (:MatchesRange_1
      (LocalVariables|SetLaunchZ 4.25e+03))
    (:MatchesRange_2
      (LocalVariables|SetLaunchZ 3.75e+03)))
  (return 0.0 0.0 (LocalVariables|GetLaunchZ)))

(fn RemoveGameplayEffectFromAllPlayers
  (params (TargetGE :type Gameplay%20Effect%20Class%20Reference))
  (LocalVariables|SetGE_LocalVar TargetGE)
  (if (Networking|IsDedicatedServer)
    (bind _outfortplayerpawns (Fortnite|GetFortPlayerPawns))
    (for _array_element _outfortplayerpawns
      (GameplayEffects|RemoveActiveGameplayEffectbySourceEffect (Ability|GetAbilitySystemComponent _array_element) (LocalVariables|GetGE_LocalVar) 0 -1))
    (Development|PrintString (Utilities|String|Append "Removed  " (Utilities|GetClassDisplayName (LocalVariables|GetGE_LocalVar)) " from all players"))))

(fn ConstructionScript
  (|CallParentFunction))
